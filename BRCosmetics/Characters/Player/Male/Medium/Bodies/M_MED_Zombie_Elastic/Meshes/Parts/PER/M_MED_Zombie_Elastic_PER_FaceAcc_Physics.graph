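# Physics asset for the face accessory part of the male medium Zombie Elastic player cosmetic, giving its eleven dyn_hair bones physical bodies and joint limits.
# Each hair bone has a small capsule, and a larger capsule covers the head.
# Every hair body is constrained directly to the head rather than chained, with moderate swing limits and very tight twist, so each can sway independently with little roll.
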PhysicsAsset /BRCosmetics/Characters/Player/Male/Medium/Bodies/M_MED_Zombie_Elastic/Meshes/Parts/PER/M_MED_Zombie_Elastic_PER_FaceAcc_Physics
bodies {
 dyn_hair_01 [{shapeType=Capsule, length=2.7908, radius=1.0964, shapeName=dyn_hair_01_capsule, center=<2 0 0>, rotation=<rot 90 0 0>}]
 dyn_hair_010 [{shapeType=Capsule, length=1.3091, radius=0.8607, shapeName=dyn_hair_010_capsule, center=<1 0 0>, rotation=<rot 90 0 0>}]
 dyn_hair_011 [{shapeType=Capsule, length=2.0431, radius=0.7739, shapeName=dyn_hair_011_capsule, center=<1 0 0>, rotation=<rot 90 0 0>}]
 dyn_hair_02 [{shapeType=Capsule, length=0.8522, radius=0.6845, shapeName=dyn_hair_02_capsule, center=<1.0082 0 0>, rotation=<rot 0 84.0876 -81.8481>}]
 dyn_hair_03 [{shapeType=Capsule, length=1.7401, radius=0.7762, shapeName=dyn_hair_03_capsule, center=<2 0 0>, rotation=<rot 90 0 0>}]
 dyn_hair_04 [{shapeType=Capsule, length=2.1581, radius=0.8128, shapeName=dyn_hair_04_capsule, center=<2 0 0>, rotation=<rot 90 0 0>}]
 dyn_hair_05 [{shapeType=Capsule, length=1.2507, radius=0.8442, shapeName=dyn_hair_05_capsule, center=<2 0 0>, rotation=<rot 90 0 0>}]
 dyn_hair_06 [{shapeType=Capsule, length=1.9187, radius=1.1586, shapeName=dyn_hair_06_capsule, center=<2 0 0>, rotation=<rot 90 0 0>}]
 dyn_hair_07 [{shapeType=Capsule, length=0.8546, radius=0.8253, shapeName=dyn_hair_07_capsule, center=<2 0 0>, rotation=<rot 90 0 0>}]
 dyn_hair_08 [{shapeType=Capsule, length=1.8869, radius=0.7158, shapeName=dyn_hair_08_capsule, center=<2 0 0>, rotation=<rot 90 0 0>}]
 dyn_hair_09 [{shapeType=Capsule, length=1.9594, radius=0.7104, shapeName=dyn_hair_09_capsule, center=<2 0 0>, rotation=<rot 90 0 0>}]
 head [{shapeType=Capsule, length=3.6445, radius=4.7198, shapeName=head_capsule, center=<5.8548 4.4225 0>, rotation=<rot 90 0 0>}]
}
constraints [
 {bone1Name=dyn_hair_01, bone2Name=head, swing1LimitDegrees=35, swing2LimitDegrees=35, twistLimitDegrees=3}
 {bone1Name=dyn_hair_02, bone2Name=head, swing1LimitDegrees=35, swing2LimitDegrees=35, twistLimitDegrees=3}
 {bone1Name=dyn_hair_03, bone2Name=head, swing1LimitDegrees=35, swing2LimitDegrees=35, twistLimitDegrees=3}
 {bone1Name=dyn_hair_04, bone2Name=head, swing1LimitDegrees=25, swing2LimitDegrees=25, twistLimitDegrees=3}
 {bone1Name=dyn_hair_05, bone2Name=head, swing1LimitDegrees=35, swing2LimitDegrees=35, twistLimitDegrees=3}
 {bone1Name=dyn_hair_06, bone2Name=head, swing1LimitDegrees=35, swing2LimitDegrees=35, twistLimitDegrees=3}
 {bone1Name=dyn_hair_07, bone2Name=head, swing1LimitDegrees=35, swing2LimitDegrees=35, twistLimitDegrees=3}
 {bone1Name=dyn_hair_08, bone2Name=head, swing1LimitDegrees=25, swing2LimitDegrees=25, twistLimitDegrees=3}
 {bone1Name=dyn_hair_09, bone2Name=head, swing1LimitDegrees=25, swing2LimitDegrees=25, twistLimitDegrees=3}
 {bone1Name=dyn_hair_010, bone2Name=head, swing1LimitDegrees=35, swing2LimitDegrees=35, twistLimitDegrees=3}
 {bone1Name=dyn_hair_011, bone2Name=head, swing1LimitDegrees=25, swing2LimitDegrees=25, twistLimitDegrees=3}
]
